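; Flag carried as a melee pickaxe weapon in Creative, showing a team color and playing speed-reactive loop audio.
; On BeginPlay the server takes the accent color from pickup instigator data; its OnRep, or remote BeginPlay, applies it to a dynamic flag material.
; While equipped, a 0.1-second timer feeds the larger of player 0's smoothed velocity and flag rotation into the loop's Speed parameter.
; Local unequip clears the timer and fades the audio.
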(event EventBeginPlay
  (bind (_instigatorhandle _instigatordata _returnvalue) (FortWeapon|GetPickupInstigatorData))
  (bind (_teamindex _targetattitude _accentcolorparam) (Utilities|Struct|BreakPickupInstigatorData _instigatordata))
  (Networking|SwitchHasAuthority
    (:Authority
      (if _returnvalue
        (Variables|Default|SetAccentColorParam _accentcolorparam)))
    (:Remote
      (CallFunction|UpdateAccentColor))))

(event FortWeapon|EventOnLocalUnEquip
  (Utilities|Time|ClearandInvalidateTimerbyHandle (Variables|Default|GetHandle%20Flag%20Loop%20Audio))
  (Audio|Components|Audio|FadeOut (Variables|Default|GetCTF_FlagLoop_Cue) 0.5 0.0 "Linear" :!class /Script/Engine.AudioComponent))

(event FortWeapon|EventOnEquippedAndAttached
  (Audio|Components|Audio|Play (Variables|Default|GetCTF_FlagLoop_Cue) 0.0)
  (bind _returnvalue (Utilities|Time|SetTimerbyEvent (EventDispatchers|CreateEvent :!event Custom|Play%20Flag%20Loop) 0.1 true))
  (Variables|Default|SetHandle%20Flag%20Loop%20Audio _returnvalue))

(event Custom|Play%20Flag%20Loop
  (bind _returnvalue (Transformation|GetActorForwardVector))
  (bind _velocity_smoothed (Variables|Default|GetVelocity%20Smoothed))
  (bind _returnvalue_2 (Math|Float|MapRangeClamped _velocity_smoothed 0.0 7e+02 0.0 1.0))
  (bind _rotation_smoothed (Variables|Default|GetRotation%20Smoothed))
  (bind _returnvalue_3 (Math|Float|MapRangeClamped _rotation_smoothed 0.0 0.5 0.0 1.0))
  (bind _returnvalue_4 (Game|GetPlayerPawn 0))
  (bind (_asfort_player_pawn _bsuccess) (Utilities|Casting|CastToFortPlayerPawn _returnvalue_4))
  ;[1 1 1] Update Falg audio based on speed and velocity of player
  (if _bsuccess
    (Variables|Default|SetVelocity%20Smoothed (Math|Interpolation|FInterpTo (Variables|Default|GetVelocity%20Smoothed) (Math|Vector|VectorLength (Transformation|GetVelocity _asfort_player_pawn)) (Utilities|Time|GetWorldDeltaSeconds) 5.0))
    (Variables|Default|SetRotation%20Smoothed (Math|Interpolation|FInterpTo (- 1.0 (Math|Float|Absolute(Float) (Math|Vector|DotProduct _returnvalue (Math|Vector|MakeVector (Variables|Default|GetCurrent%20Vector%20X) (Variables|Default|GetCurrent%20Vector%20Y) (Variables|Default|GetCurrent%20Vector%20Z))))) 0.0 (Utilities|Time|GetWorldDeltaSeconds) 5.0))
    (Variables|Default|SetCurrent%20Vector%20X (.x _returnvalue))
    (Variables|Default|SetCurrent%20Vector%20Y (.y _returnvalue))
    (Variables|Default|SetCurrent%20Vector%20Z (.z _returnvalue))
    (if (>= _returnvalue_2 _returnvalue_3)
      (Audio|Parameter|SetFloatParameter (Variables|Default|GetCTF_FlagLoop_Cue) "Speed" _returnvalue_2 :!class /Script/AudioExtensions.AudioParameterControllerInterface)
      (else
        (Audio|Parameter|SetFloatParameter (Variables|Default|GetCTF_FlagLoop_Cue) "Speed" _returnvalue_3 :!class /Script/AudioExtensions.AudioParameterControllerInterface)))
    (else
      (Audio|Parameter|SetFloatParameter (Variables|Default|GetCTF_FlagLoop_Cue) "Volume" 0.1 :!class /Script/AudioExtensions.AudioParameterControllerInterface))))

(fn OnRep_AccentColorParam
  (CallFunction|UpdateAccentColor))

(fn UpdateAccentColor
  (Networking|CanExecuteCosmeticEvents
    (:True
      (bind _flagmid (Variables|Default|GetFlagMID)
        (:then
          (Rendering|Material|SetScalarParameterValue _flagmid "TeamColor" (Variables|Default|GetAccentColorParam) :!class /Script/Engine.MaterialInstanceDynamic))
        (:else)))))

(fn ConstructionScript
  (|CallParentFunction)
  (Networking|CanExecuteCosmeticEvents
    (:True
      (bind _returnvalue (Rendering|Material|CreateDynamicMaterialInstance (Variables|Visual|GetWeaponMesh) 0 0 "None" :!class /Script/Engine.PrimitiveComponent))
      (bind _output_get (Variables|Default|SetFlagMID _returnvalue))
      (Rendering|Material|SetMaterial (Variables|Visual|GetWeaponMesh) 1 _output_get))))
