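; Gameplay ability that drops the Creative flag item by tossing it out of the player's inventory.
; Activation calls the parent and ends the ability immediately, so the toss itself runs in OnEndAbility, regardless of cancellation.
; That handler validates the avatar's Athena pawn and controller, then calls TossSpecificItem on the flag item definition.
; The toss override data takes both its distance values from the ThrowDistance scalable float.
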(event Ability|EventActivateAbility
  (|CallParentFunction)
  (Ability|EndAbility))

(event Ability|EventOnEndAbility
  (params (bWasCancelled :type Boolean))
  (bind _asfort_player_pawn_athena (Utilities|Casting|CastToFortPlayerPawnAthena (Ability|GetAvatarActorfromActorInfo)))
  (Utilities|IsValid _asfort_player_pawn_athena
    (:Is%20Valid
      (Utilities|IsValid (Fortnite|GetFortPlayerControllerfromActor _asfort_player_pawn_athena)
        (:Is%20Valid
          (Inventory|TossSpecificItem (Fortnite|GetFortPlayerControllerfromActor _asfort_player_pawn_athena) (Variables|Default|GetAGID_Flag) (Utilities|Struct|MakeFortPickupTossOverrideData true (ScalableFloat|GetValueatLevel (Variables|Default|GetThrowDistance)) (ScalableFloat|GetValueatLevel (Variables|Default|GetThrowDistance) 1.0) 0.0)))))))
