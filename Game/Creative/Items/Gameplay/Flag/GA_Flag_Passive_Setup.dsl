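; Passive setup ability for the Creative Flag item that restricts the carrier's abilities and drops the flag when the carrier goes DBNO.
; On activation it caches the pawn, binds the DBNO drop and, if there is no minigame or it is in play mode, blocks abilities by tag and hides primary HUD elements.
; Ending the ability unblocks the tags and restores the HUD without repeating the minigame check.
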(event Ability|EventActivateAbility
  (|CallParentFunction)
  (bind _asplayer_pawn_athena (Utilities|Casting|CastToPlayerPawn_Athena (Ability|GetAvatarActorfromActorInfo)))
  (Variables|Default|SetPlayerPawn _asplayer_pawn_athena)
  (CallFunction|BlockAbilitieswithTags)
  (EventDispatchers|BindEventtoOnPlayerStartDBNO (Variables|Default|GetPlayerPawn) (delegate Custom|OnPlayerStartDBNO)))

(event Custom|OnPlayerStartDBNO
  ;[1 1 1] Drop Sword on DBNO.
  (Inventory|DropSpecificItem (Fortnite|GetFortPlayerControllerfromActor (Variables|Default|GetPlayerPawn)) (Variables|Default|GetAGID_Flag) true :!class /Script/FortniteGame.FortPlayerController))

(event Ability|EventOnEndAbility
  (params (bWasCancelled :type Boolean))
  (CallFunction|UnblockAbilitieswithTags))

(event Ability|EventFailedToActivatePassiveAbility
  (params (ActorInfo :type Gameplay%20Ability%20Actor%20Info%20Structure))
  (|CallParentFunction))

(fn BlockAbilitiesWithTags
  (if (CallFunction|ShouldBlockAbilities)
    (bind _asfort_ability_system_component (Utilities|Casting|CastToFortAbilitySystemComponent (Ability|GetAbilitySystemComponent (Variables|Default|GetPlayerPawn))))
    (Tags|BlockAbilitiesWithTags _asfort_ability_system_component (select false (Variables|Default|GetBuildingAllowed) (Variables|Default|GetNoBuilding!)))
    (CallFunction|ToggleHUD false)))

(fn ShouldBlockAbilities
  (bind _playerpawn (Variables|Default|GetPlayerPawn))
  (bind _returnvalue (Minigame|GetMinigame _playerpawn :!class /Script/FortniteGame.MinigameBlueprintLibrary))
  (Utilities|IsValid _returnvalue
    (:Is%20Valid
      (return (Minigame|IsinPlayMode _returnvalue :!class /Script/FortniteGame.FortMinigame)))
    (:Is%20Not%20Valid
      (return true))))

(fn ToggleHUD
  (params (bShowHUD :type Boolean))
  (bind _asfort_player_controller_gameplay (Utilities|Casting|CastToFortPlayerControllerGameplay (Fortnite|GetFortPlayerControllerfromActor (Variables|Default|GetPlayerPawn))))
  (HUD|SetHUDElementVisibilitywithContext _asfort_player_controller_gameplay (select true (GameplayTags|MakeGameplayTagContainerfromTag "(TagName=\"HUD.Athena.HidePrimaryForAbility\")") (GameplayTags|MakeGameplayTagContainerfromTag "(TagName=\"HUD.Athena.HidePrimaryForAbility\")")) (Variables|GetbShowHUD) "None" ""))

(fn UnblockAbilitiesWithTags
  (bind _asfort_ability_system_component (Utilities|Casting|CastToFortAbilitySystemComponent (Ability|GetAbilitySystemComponent (Variables|Default|GetPlayerPawn))))
  (Tags|UnBlockAbilitiesWithTags _asfort_ability_system_component (select false (Variables|Default|GetBuildingAllowed) (Variables|Default|GetNoBuilding!)) false)
  (CallFunction|ToggleHUD true))
